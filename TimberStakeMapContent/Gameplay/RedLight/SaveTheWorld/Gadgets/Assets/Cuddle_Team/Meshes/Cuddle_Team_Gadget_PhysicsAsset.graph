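# Physics asset for the Cuddle Team gadget mesh in Save the World, giving its skeleton simple collision bodies and joints.
# Six bones each carry one capsule, with Root's being tiny next to the others.
# Five constraints form a tree: Root links to polySurface32 and upper, upper branches to guns and polySurface68, and guns links to polySurface67.
# All joints share identical swing and twist limit angles.
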PhysicsAsset /TimberStakeMapContent/Gameplay/RedLight/SaveTheWorld/Gadgets/Assets/Cuddle_Team/Meshes/Cuddle_Team_Gadget_PhysicsAsset
bodies {
 Root [{shapeType=Capsule, length=0.505, radius=0.505, rotation=<rot 0 -90 0>}]
 guns [{shapeType=Capsule, length=117.8362, radius=39.6657, center=<0.1371 -1.1879 -10.4442>, rotation=<rot 0 -90.0879 -89.3202>}]
 polySurface32 [{shapeType=Capsule, length=78.9576, radius=72.5393, center=<0.0931 2.8411 40.4867>, rotation=<rot 0 -89.8559 -89.7765>}]
 polySurface67 [{shapeType=Capsule, length=120.6088, radius=64.3561, center=<0.0007 64.5797 94.8467>, rotation=<rot 0 -89.9901 -89.9943>}]
 polySurface68 [{shapeType=Capsule, length=58.8164, radius=49.9624, center=<0.7133 4.4214 180.8764>, rotation=<rot 0 -3.3643 -58.8632>}]
 upper [{shapeType=Capsule, length=89.0144, radius=72.1035, center=<36.4967 -6.039 -2.839>, rotation=<rot 0 -173.3338 -87.1693>}]
}
constraints [
 {bone1Name=upper, bone2Name=Root, swing1LimitDegrees=45, swing2LimitDegrees=45, twistLimitDegrees=45}
 {bone1Name=guns, bone2Name=upper, swing1LimitDegrees=45, swing2LimitDegrees=45, twistLimitDegrees=45}
 {bone1Name=polySurface67, bone2Name=guns, swing1LimitDegrees=45, swing2LimitDegrees=45, twistLimitDegrees=45}
 {bone1Name=polySurface68, bone2Name=upper, swing1LimitDegrees=45, swing2LimitDegrees=45, twistLimitDegrees=45}
 {bone1Name=polySurface32, bone2Name=Root, swing1LimitDegrees=45, swing2LimitDegrees=45, twistLimitDegrees=45}
]
